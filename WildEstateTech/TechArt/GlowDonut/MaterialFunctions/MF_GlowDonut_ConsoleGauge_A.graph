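# Material function for the Glow Donut console that draws an analog gauge and blends it into input material BaseColor, also outputting color and mask.
# High-density radial gradients form the dial, outer ring and inner half-circle with outline, colored by foreground, background and line colors.
# NeedleAlpha rotates a needle texture within a clamped arc, previewing as a time-driven sine sweep.
# Subtracting the needle from a larger-scaled sample creates its outline.
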MaterialFunction /WildEstateTech/TechArt/GlowDonut/MaterialFunctions/MF_GlowDonut_ConsoleGauge_A
graph {nodes [
  Constant() {r=0.9}
  FunctionOutput(None=Multiply(A=NamedRerouteUsage(), B=NamedRerouteUsage())) {outputName=Color, sortPriority=1}
  FunctionOutput(None=NamedRerouteUsage()) {outputName=Mask, sortPriority=2}
  n1 = GetMaterialAttributes(None=FunctionInput() {bUsePreviewValueAsDefault=true, inputName=Material, inputType=FunctionInput_MaterialAttributes}) {attributeGetTypes [#1]}
  n2 = NamedRerouteUsage()
  n3 = NamedRerouteUsage()
  FunctionOutput(None=SetMaterialAttributes!(MaterialAttributes=@n1.MaterialAttributes, BaseColor=LinearInterpolate(A=@n1.BaseColor, B=@n2, Alpha=@n3)) {attributeSetTypes [#1]}) {outputName=Material}
  n4 = TextureCoordinate()
  n5 = ScalarParameter() {defaultValue=1, group=Gauge Visuals, parameterName=ScaleX}
  n6 = ScalarParameter() {defaultValue=1, group=Gauge Visuals, parameterName=ScaleY}
  NamedRerouteDeclaration(None=Multiply(
    A=FunctionInput(Preview=@n4) {bUsePreviewValueAsDefault=true, inputName=UVs, inputType=FunctionInput_Vector2, sortPriority=1}
    B=AppendVector(A=@n5, B=@n6)
   )) {name=UVs, nodeColor=<rgba 0 0.9297 1 1>}
  n7 = Subtract(A=NamedRerouteUsage()) {constA=-0.5, constB=0.5}
  n8 = Multiply(A=NamedRerouteUsage(), B=ScalarParameter() {defaultValue=0.25, group=Gauge Visuals, parameterName=NeedleCenterVOffset})
  n9 = Sine(None=Multiply(A=Time()) {constB=0.5})
  n10 = Saturate(None=FunctionInput(Preview=ConstantBiasScale(None=@n9)) {bUsePreviewValueAsDefault=true, inputName=NeedleAlpha, inputType=FunctionInput_Scalar, sortPriority=7})
  n11 = MaterialFunctionCall(
   UVs=Add(A=NamedRerouteUsage(), B=Multiply(B=@n7) {constA=-1}) {constB=0}
   Rotation Center=Add(A=MaterialFunctionCall(Y=@n8) {materialFunction=&MakeFloat2}.Result) {constB=0.5}
   "Rotation Angle (0-1)"=Clamp(None=LinearInterpolate(Alpha=@n10) {constA=0.1, constB=-0.1}) {maxDefault=0.1, minDefault=-0.1}
  ) {materialFunction=&CustomRotator}
  n12 = FunctionInput() {bUsePreviewValueAsDefault=true, inputName=ExtNeedleScalar, inputType=FunctionInput_Scalar, sortPriority=2, previewValue=<1 1 0 1>}
  n13 = Multiply(
   A=Multiply(A=NamedRerouteUsage()) {constB=2}
   B=Multiply(
    A=ScalarParameter() {defaultValue=1, desc=Changes the scale of all needles being used, group=Gauge Visuals, parameterName=NeedleScale}
    B=@n12
   )
  )
  n14 = MaterialFunctionCall(UVs=@n11.Rotated Values, Texture Scale=@n13) {materialFunction=&ScaleUVsByCenter}
  n15 = Multiply(
   A=TextureSample(UVs=@n14.UVs) {texture=/WildEstateTech/Environment/Props/Gameplay/GlowDonut/GlowDonut_MainConsole_A/Textures/T_GlowDonut_GaugeNeedle_A_M}.R
   B=@n14.0-1 mask
  )
  NamedRerouteDeclaration(None=Subtract(A=@n15, B=NamedRerouteUsage())) {name=Mask_Needle, nodeColor=<rgba 0 0.7656 1 1>}
  n16 = MaterialFunctionCall(
   UVs=@n11.Rotated Values
   Texture Scale=Multiply(A=@n13, B=ScalarParameter() {defaultValue=1.3, group=Gauge Visuals, parameterName=NeedleOutlineThickness})
  ) {materialFunction=&ScaleUVsByCenter}
  n17 = Subtract(
   A=Multiply(
    A=TextureSample(UVs=@n16.UVs) {texture=&T_GlowDonut_GaugeNeedle_A_M}.R
    B=@n16.0-1 mask
   )
   B=@n15
  )
  NamedRerouteDeclaration(None=Subtract(A=Saturate(None=@n17), B=NamedRerouteUsage())) {name=Mask_NeedleOL, nodeColor=<rgba 1 0.7031 0 1>}
  n18 = FunctionInput(Preview=Constant2Vector() {g=0.5, r=0.5}) {bUsePreviewValueAsDefault=true, inputName=Center, inputType=FunctionInput_Vector2, sortPriority=2, previewValue=<0.5 0.5 0 1>}
  NamedRerouteDeclaration(None=@n18) {name=CenterPos, nodeColor=<rgba 0.4297 0 1 1>}
  n19 = FunctionInput() {bUsePreviewValueAsDefault=true, inputName=Radius, inputType=FunctionInput_Scalar, sortPriority=3, previewValue=<0.4 0.4 0.4 1>}
  NamedRerouteDeclaration(None=@n19) {name=RadiusMain, nodeColor=<rgba 0 1 0.1094 1>}
  NamedRerouteDeclaration(None=MaterialFunctionCall(UVs=NamedRerouteUsage(), CenterPosition=NamedRerouteUsage(), Radius=NamedRerouteUsage(), Density=Constant() {r=500}) {materialFunction=&RadialGradientExponential}.RadialGradientExponential) {name=Mask_Main, nodeColor=<rgba 1 0 0.2109 1>}
  n20 = MaterialFunctionCall(
   UVs=NamedRerouteUsage()
   CenterPosition=NamedRerouteUsage()
   Radius=Multiply(A=NamedRerouteUsage(), B=ScalarParameter() {defaultValue=0.95, group=Gauge Visuals, parameterName=OuterLineThickness})
   Density=Constant() {r=500}
  ) {materialFunction=&RadialGradientExponential}
  NamedRerouteDeclaration(None=Multiply(A=OneMinus(None=@n20.RadialGradientExponential), B=NamedRerouteUsage())) {name=Mask_RingOL, nodeColor=<rgba 0 0.5312 1 1>}
  n21 = Constant() {r=1.2}
  n22 = Divide(A=NamedRerouteUsage(), B=Divide(B=@n21) {constB=4}) {constB=0.5}
  n23 = MaterialFunctionCall(
   UVs=NamedRerouteUsage()
   CenterPosition=Add(
    A=NamedRerouteUsage()
    B=MaterialFunctionCall(Y=@n22) {materialFunction=&MakeFloat2}.Result
   )
   Radius=Multiply(A=NamedRerouteUsage(), B=@n21) {constB=1.8}
   Density=Constant() {r=500}
  ) {materialFunction=&RadialGradientExponential}
  NamedRerouteDeclaration(None=Multiply(A=@n23.RadialGradientExponential, B=NamedRerouteUsage())) {name=Mask_InnerHalfCircle, nodeColor=<rgba 0.8516 1 0 1>}
  n24 = ScalarParameter() {defaultValue=0.97, desc=Half Circle Thickness, group=Gauge Visuals, parameterName=HalfCircleThickness}
  n25 = Add(A=@n22, B=Multiply(A=@n22, B=OneMinus(None=@n24)))
  n26 = MaterialFunctionCall(
   UVs=NamedRerouteUsage()
   CenterPosition=Add(
    A=NamedRerouteUsage()
    B=MaterialFunctionCall(Y=@n25) {materialFunction=&MakeFloat2}.Result
   )
   Radius=Multiply(A=NamedRerouteUsage(), B=@n21) {constB=1.8}
   Density=Constant() {r=500}
  ) {materialFunction=&RadialGradientExponential}
  n27 = NamedRerouteDeclaration(None=Multiply(A=Subtract(A=@n23.RadialGradientExponential, B=@n26.RadialGradientExponential), B=NamedRerouteUsage())) {name=Mask_HalfCircleOL, nodeColor=<rgba 0 0.5312 1 1>}
  n28 = NamedRerouteUsage()
  NamedRerouteDeclaration(None=Max(A=Max(A=@n27, B=@n28), B=NamedRerouteUsage())) {name=MergeOutlines, nodeColor=<rgba 0 1 0.5078 1>}
  n29 = Constant3Vector() {constant=<rgba 0.6724 0.6724 0.5271 1>}
  n30 = FunctionInput(Preview=Constant3Vector() {constant=<rgba 0.7758 0.0999 0.0015 1>}) {bUsePreviewValueAsDefault=true, inputName=ForegroundColor, sortPriority=9}
  n31 = LinearInterpolate(
   A=LinearInterpolate(A=FunctionInput(Preview=@n29) {bUsePreviewValueAsDefault=true, inputName=BackgroundColor, sortPriority=9}, B=@n30, Alpha=NamedRerouteUsage())
   B=@n30
   Alpha=NamedRerouteUsage()
  )
  n32 = VectorParameter() {group=Gauge Visuals, parameterName=Gauge.Color.Lines, defaultValue=<rgba 0.0312 0.0312 0.0312 1>}
  n33 = NamedRerouteUsage()
  NamedRerouteDeclaration(None=Multiply(A=LinearInterpolate(A=@n31, B=@n32.RGB, Alpha=@n33), B=NamedRerouteUsage())) {name=ColorOUT, nodeColor=<rgba 0 1 0.5078 1>}
  ScalarParameter() {defaultValue=0.4337, group=Gauge Visuals, parameterName=RadiusMain}
 ]}
