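# Material function that draws a procedural console gauge, apparently a half-disc dial with outline and needle, for the GlowDonut tech art.
# Exponential radial gradients and rounded linear gradients build the face, outline and needle masks without textures, joined through named reroutes.
# NeedleAlpha swings the needle via CustomRotator within a clamped range, previewing with a time-driven sine.
# It outputs Color, Mask, and the input Material with its BaseColor blended.
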MaterialFunction /WildEstateTech/TechArt/GlowDonut/MaterialFunctions/MF_GlowDonut_ConsoleGauge_B
graph {nodes [
  Constant() {r=0.9}
  FunctionOutput(None=Multiply(A=NamedRerouteUsage(), B=NamedRerouteUsage())) {outputName=Color, sortPriority=1}
  FunctionOutput(None=NamedRerouteUsage()) {outputName=Mask, sortPriority=2}
  n1 = GetMaterialAttributes(None=FunctionInput() {bUsePreviewValueAsDefault=true, inputName=Material, inputType=FunctionInput_MaterialAttributes}) {attributeGetTypes [#1]}
  n2 = NamedRerouteUsage()
  n3 = NamedRerouteUsage()
  FunctionOutput(None=SetMaterialAttributes!(MaterialAttributes=@n1.MaterialAttributes, BaseColor=LinearInterpolate(A=@n1.BaseColor, B=@n2, Alpha=@n3)) {attributeSetTypes [#1]}) {outputName=Material}
  n4 = FunctionInput(Preview=TextureCoordinate()) {bUsePreviewValueAsDefault=true, inputName=UVs, inputType=FunctionInput_Vector2, sortPriority=1}
  Multiply(
   A=@n4
   B=AppendVector(
    A=ScalarParameter() {defaultValue=1, group=Gauge Visuals, parameterName=ScaleX}
    B=ScalarParameter() {defaultValue=1, group=Gauge Visuals, parameterName=ScaleY}
   )
  )
  NamedRerouteDeclaration(None=@n4) {name=UVs, nodeColor=<rgba 0 0.9297 1 1>}
  n5 = NamedRerouteUsage()
  n6 = Multiply(B=Subtract(A=NamedRerouteUsage()) {constA=-0.5, constB=0.5}) {constA=-1}
  n7 = MaterialFunctionCall(Y=Multiply(A=NamedRerouteUsage(), B=ScalarParameter() {group=Gauge Visuals, parameterName=GaugeType2.NeedleCenterVOffset})) {materialFunction=&MakeFloat2}
  n8 = Time()
  n9 = ConstantBiasScale(None=Sine(None=Multiply(A=@n8) {constB=0.5}))
  n10 = LinearInterpolate(Alpha=Saturate(None=FunctionInput(Preview=@n9) {bUsePreviewValueAsDefault=true, inputName=NeedleAlpha, inputType=FunctionInput_Scalar, sortPriority=7})) {constA=0.1, constB=-0.1}
  n11 = MaterialFunctionCall(UV Channel=MaterialFunctionCall(
    UVs=Add(A=@n5, B=@n6) {constB=0}
    Rotation Center=Add(A=@n7.Result) {constB=0.5}
    "Rotation Angle (0-1)"=Clamp(None=@n10) {maxDefault=0.1, minDefault=-0.1}
   ) {materialFunction=&CustomRotator}.Rotated Values) {materialFunction=&LinearGradient}
  n12 = NamedRerouteUsage()
  n13 = Multiply(A=Constant() {r=0.04}, B=@n12)
  n14 = Multiply(A=@n12, B=Constant() {r=0.8})
  n15 = Multiply(A=Round(None=Add(A=@n11.UGradient, B=@n13)), B=Round(None=Add(A=@n11.VGradient, B=@n14)))
  n16 = OneMinus(None=@n11.UGradient)
  n17 = OneMinus(None=@n11.VGradient)
  n18 = Multiply(A=Round(None=Add(A=@n16, B=@n13)), B=Round(None=Add(A=@n17, B=@n14)))
  NamedRerouteDeclaration(None=Multiply(A=Multiply(A=@n15, B=@n18), B=NamedRerouteUsage())) {name=Needle, nodeColor=<rgba 0 0.5781 1 1>}
  n19 = MaterialFunctionCall(UV Channel=Subtract(A=NamedRerouteUsage(), B=NamedRerouteUsage())) {materialFunction=&LinearGradient}
  NamedRerouteDeclaration(None=Floor(None=OneMinus(None=@n19.VGradient))) {name=VMask, nodeColor=<rgba 1 0.5391 0 1>}
  n20 = FunctionInput(Preview=Constant2Vector() {g=0.5, r=0.5}) {bUsePreviewValueAsDefault=true, inputName=Center, inputType=FunctionInput_Vector2, sortPriority=2, previewValue=<0.5 0.5 0 1>}
  NamedRerouteDeclaration(None=@n20) {name=CenterPos, nodeColor=<rgba 0.4297 0 1 1>}
  n21 = FunctionInput(Preview=ScalarParameter() {defaultValue=0.5, group=Gauge Visuals, parameterName=RadiusMain}) {bUsePreviewValueAsDefault=true, inputName=Radius, inputType=FunctionInput_Scalar, sortPriority=3, previewValue=<0.4 0.4 0.4 1>}
  NamedRerouteDeclaration(None=@n21) {name=RadiusMain, nodeColor=<rgba 0 1 0.1094 1>}
  n22 = NamedRerouteUsage()
  n23 = NamedRerouteUsage()
  n24 = NamedRerouteUsage()
  n25 = Constant() {r=500}
  NamedRerouteDeclaration(None=Multiply(
    A=MaterialFunctionCall(UVs=@n22, CenterPosition=@n23, Radius=@n24, Density=@n25) {materialFunction=&RadialGradientExponential}.RadialGradientExponential
    B=NamedRerouteUsage()
   )) {name=Mask_Main, nodeColor=<rgba 1 0 0.2109 1>}
  n26 = MaterialFunctionCall(
   UVs=NamedRerouteUsage()
   CenterPosition=NamedRerouteUsage()
   Radius=Multiply(A=NamedRerouteUsage(), B=ScalarParameter() {defaultValue=0.95, group=Gauge Visuals, parameterName=OuterLineThickness})
   Density=Constant() {r=500}
  ) {materialFunction=&RadialGradientExponential}
  n27 = NamedRerouteUsage()
  n28 = Subtract(A=NamedRerouteUsage(), B=NamedRerouteUsage()) {constB=0.5}
  n29 = NamedRerouteUsage()
  n30 = Constant() {r=0.05}
  n31 = Ceil(None=Add(
    A=MaterialFunctionCall(UV Channel=@n28) {materialFunction=&LinearGradient}.VGradient
    B=Multiply(A=@n29, B=@n30)
   ) {constB=0.025})
  n32 = NamedRerouteUsage()
  n33 = Max(A=Multiply(A=OneMinus(None=@n26.RadialGradientExponential), B=@n27), B=Multiply(A=@n27, B=Multiply(A=@n31, B=@n32)))
  NamedRerouteDeclaration(None=Max(A=NamedRerouteDeclaration(None=@n33) {name=Mask_RingOL, nodeColor=<rgba 0 0.5312 1 1>}, B=NamedRerouteUsage())) {name=MergeOutlines, nodeColor=<rgba 0 1 0.5078 1>}
  n34 = Constant3Vector() {constant=<rgba 0.6724 0.6724 0.5271 1>}
  n35 = VectorParameter() {group=Gauge Visuals, parameterName=Gauge.Color.Lines, defaultValue=<rgba 0.0312 0.0312 0.0312 1>}
  n36 = NamedRerouteUsage()
  NamedRerouteDeclaration(None=Multiply(A=LinearInterpolate(A=@n34, B=@n35.RGB, Alpha=@n36), B=NamedRerouteUsage())) {name=ColorOUT, nodeColor=<rgba 0 1 0.5078 1>}
  NamedRerouteUsage()
  n37 = NamedRerouteUsage()
  Round(None=OneMinus(None=MaterialFunctionCall(UV Channel=@n37) {materialFunction=&LinearGradient}.VGradient))
 ]}
