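# Material function for a GlowDonut console gauge variant that draws a swinging needle, outputting a line color parameter and a mask.
# UVs shift by center and rotate around a height-offset pivot, with NeedleAlpha mapped to a narrow clamped angle.
# Rounded, mirrored linear gradients form the needle rectangle from width and length parameters, multiplied by what appears to be a vertical clipping mask.
# NeedleAlpha defaults to a time-driven sine sweep.
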MaterialFunction /WildEstateTech/TechArt/GlowDonut/MaterialFunctions/MF_GlowDonut_ConsoleGauge_C
graph {nodes [
  Constant() {r=0.9}
  FunctionOutput(None=VectorParameter() {group=Gauge Visuals, parameterName=Gauge.Color.Lines, defaultValue=<rgba 0.0312 0.0312 0.0312 1>}.RGB) {outputName=Color, sortPriority=1}
  FunctionOutput(None=NamedRerouteUsage()) {outputName=Mask, sortPriority=2}
  n1 = ScalarParameter() {defaultValue=0.02, group=Gauge Visuals, parameterName=GaugeType3.Width}
  Multiply(A=@n1)
  NamedRerouteDeclaration(None=FunctionInput(Preview=TextureCoordinate()) {bUsePreviewValueAsDefault=true, inputName=UVs, inputType=FunctionInput_Vector2, sortPriority=1}) {name=UVs, nodeColor=<rgba 0 0.9297 1 1>}
  n2 = NamedRerouteUsage()
  n3 = Multiply(B=Subtract(A=NamedRerouteUsage()) {constA=-0.5, constB=0.5}) {constA=-1}
  n4 = FunctionInput() {bUsePreviewValueAsDefault=true, inputName=NeedleHeightOffset, inputType=FunctionInput_Scalar, sortPriority=2, previewValue=<0.5 0.5 0 1>}
  n5 = MaterialFunctionCall(Y=@n4) {materialFunction=&MakeFloat2}
  n6 = Time()
  n7 = ConstantBiasScale(None=Sine(None=Multiply(A=@n6) {constB=0.5}))
  n8 = LinearInterpolate(Alpha=Saturate(None=FunctionInput(Preview=@n7) {bUsePreviewValueAsDefault=true, inputName=NeedleAlpha, inputType=FunctionInput_Scalar, sortPriority=7})) {constA=0.1, constB=-0.1}
  n9 = MaterialFunctionCall(UV Channel=MaterialFunctionCall(
    UVs=Add(A=@n2, B=@n3) {constB=0}
    Rotation Center=Add(A=@n5.Result) {constB=0.5}
    "Rotation Angle (0-1)"=Clamp(None=@n8) {maxDefault=0.1, minDefault=-0.1}
   ) {materialFunction=&CustomRotator}.Rotated Values) {materialFunction=&LinearGradient}
  n10 = ScalarParameter() {defaultValue=0.4, group=Gauge Visuals, parameterName=GaugeType3.Length}
  n11 = Multiply(A=Round(None=Add(A=@n9.UGradient, B=@n1)), B=Round(None=Add(A=@n9.VGradient, B=@n10)))
  n12 = OneMinus(None=@n9.UGradient)
  n13 = OneMinus(None=@n9.VGradient)
  n14 = Multiply(A=Round(None=Add(A=@n12, B=@n1)), B=Round(None=Add(A=@n13, B=@n10)))
  NamedRerouteDeclaration(None=Multiply(A=Multiply(A=@n11, B=@n14), B=NamedRerouteUsage())) {name=Needle, nodeColor=<rgba 0 0.5781 1 1>}
  n15 = MaterialFunctionCall(UV Channel=Subtract(A=NamedRerouteUsage(), B=NamedRerouteUsage())) {materialFunction=&LinearGradient}
  NamedRerouteDeclaration(None=Floor(None=OneMinus(None=@n15.VGradient))) {name=VMask, nodeColor=<rgba 1 0.5391 0 1>}
  n16 = FunctionInput(Preview=Constant2Vector() {g=0.5, r=0.5}) {bUsePreviewValueAsDefault=true, inputName=Center, inputType=FunctionInput_Vector2, sortPriority=2, previewValue=<0.5 0.5 0 1>}
  NamedRerouteDeclaration(None=@n16) {name=CenterPos, nodeColor=<rgba 0.4297 0 1 1>}
  n17 = NamedRerouteUsage()
  Round(None=OneMinus(None=MaterialFunctionCall(UV Channel=@n17) {materialFunction=&LinearGradient}.VGradient))
 ]}
